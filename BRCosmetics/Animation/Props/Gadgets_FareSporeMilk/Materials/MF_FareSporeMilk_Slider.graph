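# Material function that draws a gradient-filled, outlined slider bar into a material's base color, emissive and opacity mask, for the FareSporeMilk gadget cosmetic.
# The pill shape combines MF_SquareMaskGenerator with two SphereMask end caps, with Slider_Value moving the right end and an inset copy cutting the outline.
# An optional static-switch branch modulates brightness by Amplitude outside the slider masks and pulses emissive on a fixed time-based sine.
# The slider mask is also output.
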MaterialFunction /BRCosmetics/Animation/Props/Gadgets_FareSporeMilk/Materials/MF_FareSporeMilk_Slider
graph {nodes [
  Constant() {r=0.676}
  n1 = GetMaterialAttributes(None=FunctionInput() {inputType=FunctionInput_MaterialAttributes}) {attributeGetTypes [#1, #2, #3]}
  n2 = TextureCoordinate()
  n3 = Constant() {r=0.22}
  n4 = Constant() {r=0.57}
  n5 = FunctionInput(Preview=Constant() {r=1}) {bUsePreviewValueAsDefault=true, inputName=Slider_Value, inputType=FunctionInput_Scalar, sortPriority=10}
  n6 = LinearInterpolate(A=@n3, B=@n4, Alpha=@n5) {constA=0.25, constB=0.57}
  n7 = FunctionInput(Preview=Constant() {r=0.7005}) {bUsePreviewValueAsDefault=true, inputName=Slider_PosY, inputType=FunctionInput_Scalar, sortPriority=1}
  n8 = FunctionInput(Preview=Constant() {r=0.0245}) {bUsePreviewValueAsDefault=true, inputName=Slider_Thickness, inputType=FunctionInput_Scalar, sortPriority=1}
  n9 = Subtract(A=@n7, B=@n8)
  n10 = Add(A=@n7, B=@n8)
  n11 = Multiply(A=Add(A=@n9, B=@n10)) {constB=0.5}
  n12 = SphereMask(A=@n2, B=AppendVector(A=@n6, B=@n11), Radius=@n8) {attenuationRadius=0.02}
  n13 = Constant() {r=0.182}
  n14 = MaterialFunctionCall!(
   Texture Coordinate=TextureCoordinate()
   X Pos Right=@n6
   X Pos Left=@n13
   Y Pos Up=@n9
   Y Pos Down=@n10
   SquareSoftness=Constant()
   RotateSquare
   RotationAmount
  ) {materialFunction=/Game/Accessories/FORT_Backpacks/Backpack_F_MED_GroovyReader_Play/Materials/MaterialFunctions/MF_SquareMaskGenerator}
  n15 = AppendVector(A=@n13, B=@n11)
  n16 = Saturate(None=Add(A=Add(A=@n12, B=@n14.SquareMask), B=SphereMask(A=@n2, B=@n15, Radius=@n8) {attenuationRadius=0.02}))
  n17 = OneMinus(None=Add(A=FunctionInput() {bUsePreviewValueAsDefault=true, inputName=SliderMask, inputType=FunctionInput_Scalar}, B=@n16))
  n18 = FunctionInput() {bUsePreviewValueAsDefault=true, inputName=Amplitude, inputType=FunctionInput_Scalar}
  n19 = Multiply(A=Saturate(None=Add(A=@n17) {constB=0.55}), B=Multiply(A=@n18) {constB=2.5})
  n20 = ComponentMask(None=TextureCoordinate()) {r=true}
  n21 = Constant() {r=0.2}
  n22 = Constant() {r=0.8}
  n23 = TextureSample(UVs=Saturate(None=MaterialFunctionCall(Input=@n20, Input Low=@n21, Input High=@n22) {materialFunction=&RemapValue_0-1}.Result)) {
   samplerSource=SSM_Clamp_WorldGroupSettings
   samplerType=SAMPLERTYPE_LinearColor
   texture=/BRCosmetics/Animation/Props/Gadgets_FareSporeMilk/Textures/T_FareSpore_Milk_GradientMap
  }
  n24 = Constant() {r=-1}
  n25 = TextureCoordinate()
  n26 = FunctionInput(Preview=Constant() {r=0.002}) {bUsePreviewValueAsDefault=true, inputName=Sldier_OutlineThickness, inputType=FunctionInput_Scalar, sortPriority=5}
  n27 = LinearInterpolate(A=Subtract(A=@n3, B=@n26), B=Subtract(A=@n4, B=@n26), Alpha=@n5) {constA=0.25, constB=0.57}
  n28 = Add(A=@n26, B=@n9)
  n29 = Subtract(A=@n10, B=@n26)
  n30 = Multiply(A=Add(A=@n28, B=@n29)) {constB=0.5}
  n31 = Subtract(A=@n8, B=@n26)
  n32 = SphereMask(A=@n25, B=AppendVector(A=@n27, B=@n30), Radius=@n31) {attenuationRadius=0.02}
  n33 = Add(A=@n26, B=Constant() {r=0.182})
  n34 = MaterialFunctionCall(Texture Coordinate=TextureCoordinate(), X Pos Right=@n27, X Pos Left=@n33, Y Pos Up=@n28, Y Pos Down=@n29, SquareSoftness=Constant()) {materialFunction=&MF_SquareMaskGenerator}
  n35 = AppendVector(A=@n33, B=@n30)
  n36 = Saturate(None=Add(A=Add(A=@n32, B=@n34.SquareMask), B=SphereMask(A=@n25, B=@n35, Radius=@n31) {attenuationRadius=0.02}))
  n37 = LinearInterpolate(A=Multiply(A=Desaturation(None=@n23.RGB, Fraction=@n24), B=@n16), Alpha=Saturate(None=Subtract(A=@n16, B=@n36)))
  n38 = Add(A=@n1.EmissiveColor, B=Multiply(B=@n37) {constA=0.3})
  n39 = Multiply(A=@n19, B=@n38) {constB=6}
  n40 = Multiply(B=@n38) {constA=0.9}
  n41 = Time()
  n42 = Sine(None=Multiply(A=Divide(A=@n41) {constB=4}, B=Divide() {constA=128, constB=60}))
  n43 = SmoothStep(Value=Add(A=Multiply(A=@n42) {constB=0.5}) {constB=0.5}) {constMax=0.6, constMin=0.4}
  n44 = Multiply(A=Desaturation(None=Add(A=@n39, B=@n40), Fraction=Constant()), B=Add(A=Multiply(A=@n43) {constB=0.5}))
  n45 = FunctionInput(Preview=StaticBool()) {bUsePreviewValueAsDefault=true, inputName=UseAudioReactivity, inputType=FunctionInput_StaticBool}
  n46 = Add(A=@n1.BaseColor, B=Multiply(A=@n37) {constB=0.3})
  n47 = Multiply(A=Max(A=@n18) {constB=0.5}, B=@n46)
  n48 = Add(A=Multiply(A=@n19, B=@n47), B=@n47)
  FunctionOutput(None=SetMaterialAttributes!(
    MaterialAttributes=@n1.MaterialAttributes
    EmissiveColor=StaticSwitch(True=@n44, False=@n38, Value=@n45)
    OpacityMask=Add(A=@n1.OpacityMask, B=@n16)
    BaseColor=StaticSwitch(True=@n48, False=@n46, Value=@n45)
   ) {attributeSetTypes [#1, #2, #3]})
  FunctionOutput(None=@n16) {outputName=SliderMask}
  Multiply(A=Subtract(A=@n29, B=@n28)) {constB=0.5}
 ]}
