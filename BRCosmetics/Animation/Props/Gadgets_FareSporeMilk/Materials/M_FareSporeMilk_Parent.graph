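# Parent material for the FareSporeMilk gadget prop, layering an animated, audio-reactive music-panel look over the 2023 character parent material functions.
# Square UV masks and texture masks isolate equalizer, dot and panel regions for hue shifts, blinking and emissive glow, with stepped time at 128 BPM supplying the beat.
# Audio amplitude and spectrum feed a waveform-string equalizer, and two chained slider functions add parameter-set sliders before the PostFX stage.
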Material /BRCosmetics/Animation/Props/Gadgets_FareSporeMilk/Materials/M_FareSporeMilk_Parent
graph {
 expressions {
  MaterialExpressionAbs_1 {class=MaterialExpressionAbs, inputs [None {from=MaterialExpressionSubtract_8, output=""}], properties {}}
  MaterialExpressionAbs_3 {class=MaterialExpressionAbs, inputs [None {from=MaterialExpressionSubtract_3, output=""}], properties {}}
  MaterialExpressionAbs_4 {class=MaterialExpressionAbs, inputs [None {from=MaterialExpressionOneMinus_14, output=""}], properties {}}
  MaterialExpressionAdd_1 {
   class=MaterialExpressionAdd
   inputs [
    A {from=MaterialExpressionTextureCoordinate_1}
    B {from=MaterialExpressionConstant2Vector_0}
   ]
  }
  MaterialExpressionAdd_10 {
   class=MaterialExpressionAdd
   inputs [
    A {from=MaterialExpressionMultiply_21}
    B {from=MaterialExpressionNamedRerouteUsage_7}
   ]
  }
  MaterialExpressionAdd_11 {
   class=MaterialExpressionAdd
   inputs [
    A {from=MaterialExpressionMaterialFunctionCall_10, output=SliderMask}
    B {from=MaterialExpressionNamedRerouteUsage_25}
   ]
  }
  MaterialExpressionAdd_14 {
   class=MaterialExpressionAdd
   inputs [
    A {from=MaterialExpressionMultiply_24}
    B {from=""}
   ]
   properties {constB=0.5}
  }
  MaterialExpressionAdd_17 {
   class=MaterialExpressionAdd
   inputs [
    A {from=MaterialExpressionMultiply_71}
    B {from=MaterialExpressionMultiply_61}
   ]
  }
  MaterialExpressionAdd_18 {
   class=MaterialExpressionAdd
   inputs [
    A {from=MaterialExpressionMultiply_55}
    B {from=""}
   ]
  }
  MaterialExpressionAdd_2 {
   class=MaterialExpressionAdd
   inputs [
    A {from=MaterialExpressionFrac_1}
    B {from=MaterialExpressionMultiply_55}
   ]
  }
  MaterialExpressionAdd_22 {
   class=MaterialExpressionAdd
   inputs [
    A {from=MaterialExpressionNamedRerouteUsage_2}
    B {from=MaterialExpressionNamedRerouteUsage_46}
   ]
  }
  MaterialExpressionAdd_23 {
   class=MaterialExpressionAdd
   inputs [
    A {from=MaterialExpressionTextureCoordinate_10}
    B {from=MaterialExpressionConstant2Vector_2}
   ]
  }
  MaterialExpressionAdd_24 {
   class=MaterialExpressionAdd
   inputs [
    A {from=MaterialExpressionComponentMask_12}
    B {from=""}
   ]
   properties {constB=2}
  }
  MaterialExpressionAdd_26 {
   class=MaterialExpressionAdd
   inputs [
    A {from=MaterialExpressionMultiply_74}
    B {from=""}
   ]
   properties {constB=0.5}
  }
  MaterialExpressionAdd_27 {
   class=MaterialExpressionAdd
   inputs [
    A {from=MaterialExpressionOneMinus_20}
    B {from=""}
   ]
   properties {constB=0.5}
  }
  MaterialExpressionAdd_3 {
   class=MaterialExpressionAdd
   inputs [
    A {from=MaterialExpressionMultiply_8}
    B {from=""}
   ]
   properties {constB=-1}
  }
  MaterialExpressionAdd_4 {
   class=MaterialExpressionAdd
   inputs [
    A {from=MaterialExpressionMaterialFunctionCall_6, output=Result}
    B {from=MaterialExpressionNamedRerouteUsage_46}
   ]
  }
  MaterialExpressionAdd_5 {
   class=MaterialExpressionAdd
   inputs [
    A {from=MaterialExpressionOneMinus_1}
    B {from=""}
   ]
   properties {constB=0.3}
  }
  MaterialExpressionAdd_7 {
   class=MaterialExpressionAdd
   inputs [
    A {from=MaterialExpressionLinearInterpolate_7}
    B {from=""}
   ]
  }
  MaterialExpressionAdd_8 {
   class=MaterialExpressionAdd
   inputs [
    A {from=MaterialExpressionMultiply_12}
    B {from=MaterialExpressionMultiply_10}
   ]
  }
  MaterialExpressionAdd_9 {
   class=MaterialExpressionAdd
   inputs [
    A {from=MaterialExpressionOneMinus_5}
    B {from=MaterialExpressionMultiply_16}
   ]
  }
  MaterialExpressionComponentMask_0 {class=MaterialExpressionComponentMask, inputs [None {from=MaterialExpressionMultiply_6}], properties {g=true}}
  MaterialExpressionComponentMask_1 {class=MaterialExpressionComponentMask, inputs [None {from=MaterialExpressionNamedRerouteUsage_10}], properties {g=true}}
  MaterialExpressionComponentMask_12 {class=MaterialExpressionComponentMask, inputs [None {from=MaterialExpressionTextureCoordinate_12}], properties {g=true}}
  MaterialExpressionComponentMask_3 {class=MaterialExpressionComponentMask, inputs [None {from=MaterialExpressionTextureCoordinate_11}], properties {g=true}}
  MaterialExpressionComponentMask_4 {class=MaterialExpressionComponentMask, inputs [None {from=MaterialExpressionTextureCoordinate_12}], properties {g=true}}
  MaterialExpressionConstant2Vector_0 {class=MaterialExpressionConstant2Vector, properties {g=0.288, r=-0.425}}
  MaterialExpressionConstant2Vector_1 {class=MaterialExpressionConstant2Vector, properties {g=0.32, r=1}}
  MaterialExpressionConstant2Vector_2 {class=MaterialExpressionConstant2Vector, properties {g=0.1, r=-2.28}}
  MaterialExpressionConstant_0 {class=MaterialExpressionConstant, properties {r=0.992}}
  MaterialExpressionConstant_1 {class=MaterialExpressionConstant, properties {r=0.42}}
  MaterialExpressionConstant_10 {class=MaterialExpressionConstant, properties {r=0.64}}
  MaterialExpressionConstant_11 {class=MaterialExpressionConstant, properties {r=0.97}}
  MaterialExpressionConstant_12 {class=MaterialExpressionConstant, properties {r=0.02}}
  MaterialExpressionConstant_13 {class=MaterialExpressionConstant, properties {r=0.8}}
  MaterialExpressionConstant_14 {class=MaterialExpressionConstant, properties {r=0.992}}
  MaterialExpressionConstant_15 {class=MaterialExpressionConstant, properties {r=0.99}}
  MaterialExpressionConstant_16 {class=MaterialExpressionConstant, properties {r=0.636}}
  MaterialExpressionConstant_17 {class=MaterialExpressionConstant, properties {r=0.55}}
  MaterialExpressionConstant_19 {class=MaterialExpressionConstant, properties {r=0.61}}
  MaterialExpressionConstant_2 {class=MaterialExpressionConstant, properties {r=-1}}
  MaterialExpressionConstant_21 {class=MaterialExpressionConstant, properties {r=-1}}
  MaterialExpressionConstant_3 {class=MaterialExpressionConstant, properties {r=0.992}}
  MaterialExpressionConstant_4 {class=MaterialExpressionConstant, properties {r=0.3873}}
  MaterialExpressionConstant_5 {class=MaterialExpressionConstant, properties {r=0.83}}
  MaterialExpressionConstant_6 {class=MaterialExpressionConstant, properties {r=0.0451}}
  MaterialExpressionConstant_7 {class=MaterialExpressionConstant, properties {r=0.83}}
  MaterialExpressionConstant_8 {class=MaterialExpressionConstant, properties {r=0.03}}
  MaterialExpressionDesaturation_0 {
   class=MaterialExpressionDesaturation
   inputs [
    None {from=MaterialExpressionAdd_4, output=""}
    Fraction {from=MaterialExpressionConstant_2, output=""}
   ]
   properties {}
  }
  MaterialExpressionDesaturation_1 {
   class=MaterialExpressionDesaturation
   inputs [
    None {from=MaterialExpressionMaterialFunctionCall_12, output=Result}
    Fraction {from=MaterialExpressionConstant_21, output=""}
   ]
   properties {}
  }
  MaterialExpressionDesaturation_2 {
   class=MaterialExpressionDesaturation
   inputs [
    None {from=MaterialExpressionMultiply_12, output=""}
    Fraction {from=MaterialExpressionMultiply_11, output=""}
   ]
   properties {}
  }
  MaterialExpressionDivide_0 {
   class=MaterialExpressionDivide
   inputs [
    A {from=MaterialExpressionFloor_0}
    B {from=""}
   ]
   properties {constB=3}
  }
  MaterialExpressionDivide_2 {
   class=MaterialExpressionDivide
   inputs [
    A {from=MaterialExpressionAbs_3}
    B {from=MaterialExpressionOneMinus_15}
   ]
  }
  MaterialExpressionDivide_4 {
   class=MaterialExpressionDivide
   inputs [
    A {from=MaterialExpressionFloor_1}
    B {from=""}
   ]
  }
  MaterialExpressionDivide_6 {
   class=MaterialExpressionDivide
   inputs [
    A {from=""}
    B {from=""}
   ]
   properties {constA=128, constB=60}
  }
  MaterialExpressionDivide_7 {
   class=MaterialExpressionDivide
   inputs [
    A {from=MaterialExpressionFloor_3}
    B {from=""}
   ]
   properties {constB=4}
  }
  MaterialExpressionDivide_8 {
   class=MaterialExpressionDivide
   inputs [
    A {from=""}
    B {from=""}
   ]
   properties {constA=128, constB=60}
  }
  MaterialExpressionFloor_0 {class=MaterialExpressionFloor, inputs [None {from=MaterialExpressionMultiply_4, output=""}], properties {}}
  MaterialExpressionFloor_1 {class=MaterialExpressionFloor, inputs [None {from=MaterialExpressionMultiply_4, output=""}], properties {}}
  MaterialExpressionFloor_3 {class=MaterialExpressionFloor, inputs [None {from=MaterialExpressionMultiply_30, output=""}], properties {}}
  MaterialExpressionFrac_1 {class=MaterialExpressionFrac, inputs [None {from=MaterialExpressionComponentMask_0, output=""}], properties {}}
  MaterialExpressionFrac_2 {class=MaterialExpressionFrac, inputs [None {from=MaterialExpressionDivide_7, output=""}], properties {}}
  MaterialExpressionFrac_3 {class=MaterialExpressionFrac, inputs [None {from=MaterialExpressionDivide_4, output=""}], properties {}}
  MaterialExpressionGetMaterialAttributes_1 {
   class=MaterialExpressionGetMaterialAttributes
   inputs [None {from=MaterialExpressionMaterialFunctionCall_2, output="Switch Param (False)"}]
   properties {attributeGetTypes [#1, #2, #3]}
  }
  MaterialExpressionLinearInterpolate_0 {
   class=MaterialExpressionLinearInterpolate
   inputs [
    A {from=""}
    B {from=""}
    Alpha {from=MaterialExpressionSaturate_4}
   ]
   properties {constA=0.007, constB=0.1}
  }
  MaterialExpressionLinearInterpolate_1 {
   class=MaterialExpressionLinearInterpolate
   inputs [
    A {from=MaterialExpressionScalarParameter_27}
    B {from=""}
    Alpha {from=MaterialExpressionNamedRerouteUsage_48}
   ]
   properties {constB=0.12}
  }
  MaterialExpressionLinearInterpolate_2 {
   class=MaterialExpressionLinearInterpolate
   inputs [
    A {from=MaterialExpressionNamedRerouteUsage_9}
    B {from=MaterialExpressionNamedRerouteUsage_17}
    Alpha {from=MaterialExpressionNamedRerouteUsage_16}
   ]
  }
  MaterialExpressionLinearInterpolate_7 {
   class=MaterialExpressionLinearInterpolate
   inputs [
    A {from=""}
    B {from=MaterialExpressionScalarParameter_4}
    Alpha {from=MaterialExpressionNamedRerouteUsage_24}
   ]
   properties {constA=0.01}
  }
  MaterialExpressionMaterialFunctionCall_0 {
   class=MaterialExpressionMaterialFunctionCall
   inputs [
    Overide Texture Set UVs {from=""}
    Override Tech Art Mask UVs {from=""}
    Override Basic Emissive UVs {from=""}
   ]
   properties {materialFunction=&MF_CharacterParent_2023_PreFX}
  }
  MaterialExpressionMaterialFunctionCall_1 {
   class=MaterialExpressionMaterialFunctionCall
   inputs [
    Material {from=MaterialExpressionMaterialFunctionCall_11, output=Result}
    M {from=MaterialExpressionMaterialFunctionCall_0, output=M}
    Pre Skinned Local Position {from=MaterialExpressionMaterialFunctionCall_0, output=M}
    Distance Blend {from=MaterialExpressionMaterialFunctionCall_0, output=M}
    Fresnel {from=MaterialExpressionMaterialFunctionCall_0, output=M}
   ]
   properties {materialFunction=&MF_CharacterParent_2023_PostFX}
  }
  MaterialExpressionMaterialFunctionCall_10 {
   class=MaterialExpressionMaterialFunctionCall
   inputs [
    Amplitude {from=""}
    UseAudioReactivity {from=""}
    SliderMask {from=""}
    In {from=MaterialExpressionNamedRerouteUsage_11}
    Slider_PosY {from=MaterialExpressionConstant_19}
    Slider_Thickness {from=""}
    Sldier_OutlineThickness {from=""}
    Slider_Value {from=MaterialExpressionScalarParameter_0}
   ]
   properties {materialFunction=&/MF_FareSporeMilk_Slider}
  }
  MaterialExpressionMaterialFunctionCall_11 {
   class=MaterialExpressionMaterialFunctionCall
   inputs [
    Amplitude {from=MaterialExpressionSaturate_5}
    UseAudioReactivity {from=MaterialExpressionStaticBool_1}
    SliderMask {from=MaterialExpressionAdd_11}
    In {from=MaterialExpressionMaterialFunctionCall_10, output=Result}
    Slider_PosY {from=""}
    Slider_Thickness {from=""}
    Sldier_OutlineThickness {from=""}
    Slider_Value {from=MaterialExpressionScalarParameter_1}
   ]
   properties {materialFunction=&MF_FareSporeMilk_Slider}
  }
  MaterialExpressionMaterialFunctionCall_12 {
   class=MaterialExpressionMaterialFunctionCall
   inputs [
    Hue Shift Percentage {from=MaterialExpressionNamedRerouteUsage_14}
    Texture {from=MaterialExpressionGetMaterialAttributes_1, output=Base Color}
   ]
   properties {materialFunction=&HueShift}
  }
  MaterialExpressionMaterialFunctionCall_14 {
   class=MaterialExpressionMaterialFunctionCall
   inputs []
   properties {materialFunction=/Game/Packages/Fortress_SharedMaterials/Base_Material_Functions/BaseCharacterShaderFunctions/Audio/MF_AudioReactivity}
  }
  MaterialExpressionMaterialFunctionCall_15 {
   class=MaterialExpressionMaterialFunctionCall
   inputs [Float4 {from=MaterialExpressionPower_3}]
   properties {materialFunction=&BreakOutFloat4Components}
  }
  MaterialExpressionMaterialFunctionCall_17 {
   class=MaterialExpressionMaterialFunctionCall
   inputs [
    Texture Coordinate {from=MaterialExpressionTextureCoordinate_6}
    X Pos Right {from=MaterialExpressionConstant_0}
    X Pos Left {from=MaterialExpressionConstant_7}
    Y Pos Up {from=MaterialExpressionConstant_8}
    Y Pos Down {from=MaterialExpressionConstant_1}
    SquareSoftness {from=""}
    RotateSquare {from=""}
    RotationAmount {from=""}
   ]
   properties {materialFunction=/Game/Accessories/FORT_Backpacks/Backpack_F_MED_GroovyReader_Play/Materials/MaterialFunctions/MF_SquareMaskGenerator}
  }
  MaterialExpressionMaterialFunctionCall_2 {
   class=MaterialExpressionMaterialFunctionCall
   inputs [Result {from=MaterialExpressionNamedRerouteUsage_12}]
   properties {materialFunction=/Game/Global/Materials/Character_Parent_2023/Functions/MF_CharacterParent_2023_OpacityMask}
  }
  MaterialExpressionMaterialFunctionCall_21 {
   class=MaterialExpressionMaterialFunctionCall
   inputs [
    Audio Amplitude {from=MaterialExpressionNamedRerouteUsage_34}
    Audio Spectrum {from=MaterialExpressionPower_2}
    Time {from=""}
    Freq Base {from=MaterialExpressionScalarParameter_36}
    Coordinate Gradient {from=MaterialExpressionMultiply_54}
    Freq Curve Factor {from=MaterialExpressionScalarParameter_13}
    Amplitude Curve Factor {from=MaterialExpressionScalarParameter_30}
    Amplitude Base {from=MaterialExpressionScalarParameter_29}
    Scroll Base {from=MaterialExpressionScalarParameter_31}
    Scroll Curve Factor {from=MaterialExpressionScalarParameter_32}
   ]
   properties {materialFunction=/BRCosmetics/Characters/Player/Female/Medium/Bodies/F_MED_Tactical_Red_Punk/Skins/Disco/Materials/MF_AudioVisualizationString}
  }
  MaterialExpressionMaterialFunctionCall_3 {
   class=MaterialExpressionMaterialFunctionCall
   inputs [
    InAmplitude {from=""}
    InSpectrum {from=""}
    CustomFakeSpectrum {from=""}
    CustomFakeAmplitude {from=""}
   ]
   properties {materialFunction=/BRCosmetics/Global/MaterialFunctions/AudioReactivity/MF_AudioReactivity_FrontendLoop}
  }
  MaterialExpressionMaterialFunctionCall_4 {
   class=MaterialExpressionMaterialFunctionCall
   inputs [
    Texture Coordinate {from=MaterialExpressionTextureCoordinate_0}
    X Pos Right {from=MaterialExpressionConstant_3}
    X Pos Left {from=MaterialExpressionConstant_5}
    Y Pos Up {from=MaterialExpressionConstant_6}
    Y Pos Down {from=MaterialExpressionConstant_4}
    SquareSoftness {from=""}
    RotateSquare {from=""}
    RotationAmount {from=""}
   ]
   properties {materialFunction=&MF_SquareMaskGenerator}
  }
  MaterialExpressionMaterialFunctionCall_5 {
   class=MaterialExpressionMaterialFunctionCall
   inputs [
    Texture Coordinate {from=MaterialExpressionTextureCoordinate_2}
    X Pos Right {from=MaterialExpressionConstant_10}
    X Pos Left {from=MaterialExpressionConstant_12}
    Y Pos Up {from=MaterialExpressionConstant_13}
    Y Pos Down {from=MaterialExpressionConstant_11}
    SquareSoftness {from=""}
    RotateSquare {from=""}
    RotationAmount {from=""}
   ]
   properties {materialFunction=&MF_SquareMaskGenerator}
  }
  MaterialExpressionMaterialFunctionCall_6 {
   class=MaterialExpressionMaterialFunctionCall
   inputs [
    Hue Shift Percentage {from=MaterialExpressionNamedRerouteUsage_13}
    Texture {from=MaterialExpressionNamedRerouteUsage_19}
   ]
   properties {materialFunction=&HueShift}
  }
  MaterialExpressionMaterialFunctionCall_7 {
   class=MaterialExpressionMaterialFunctionCall
   inputs [
    Texture Coordinate {from=MaterialExpressionTextureCoordinate_3}
    X Pos Right {from=MaterialExpressionConstant_14}
    X Pos Left {from=MaterialExpressionConstant_16}
    Y Pos Up {from=MaterialExpressionConstant_17}
    Y Pos Down {from=MaterialExpressionConstant_15}
    SquareSoftness {from=""}
    RotateSquare {from=""}
    RotationAmount {from=""}
   ]
   properties {materialFunction=&MF_SquareMaskGenerator}
  }
  MaterialExpressionMaterialFunctionCall_8 {
   class=MaterialExpressionMaterialFunctionCall
   inputs [
    UVs {from=MaterialExpressionAdd_1}
    Texture Scale {from=MaterialExpressionConstant2Vector_1}
   ]
   properties {materialFunction=&ScaleUVsByCenter}
  }
  MaterialExpressionMaterialFunctionCall_9 {
   class=MaterialExpressionMaterialFunctionCall
   inputs [
    XYZ {from=""}
    XY {from=MaterialExpressionAdd_23}
   ]
   properties {materialFunction=&Swizzle}
  }
  MaterialExpressionMax_0 {
   class=MaterialExpressionMax
   inputs [
    A {from=MaterialExpressionMultiply_0}
    B {from=""}
   ]
   properties {constB=0.7}
  }
  MaterialExpressionMultiply_0 {
   class=MaterialExpressionMultiply
   inputs [
    A {from=MaterialExpressionPower_0}
    B {from=""}
   ]
   properties {constB=6}
  }
  MaterialExpressionMultiply_10 {
   class=MaterialExpressionMultiply
   inputs [
    A {from=MaterialExpressionMultiply_7}
    B {from=""}
   ]
   properties {constB=2}
  }
  MaterialExpressionMultiply_11 {
   class=MaterialExpressionMultiply
   inputs [
    A {from=MaterialExpressionComponentMask_1}
    B {from=""}
   ]
   properties {constB=0}
  }
  MaterialExpressionMultiply_12 {
   class=MaterialExpressionMultiply
   inputs [
    A {from=MaterialExpressionDesaturation_0}
    B {from=MaterialExpressionSaturate_0}
   ]
  }
  MaterialExpressionMultiply_14 {
   class=MaterialExpressionMultiply
   inputs [
    A {from=MaterialExpressionMultiply_68}
    B {from=""}
   ]
   properties {constB=2}
  }
  MaterialExpressionMultiply_16 {
   class=MaterialExpressionMultiply
   inputs [
    A {from=MaterialExpressionSwitch_0}
    B {from=MaterialExpressionNamedRerouteUsage_5}
   ]
  }
  MaterialExpressionMultiply_2 {
   class=MaterialExpressionMultiply
   inputs [
    A {from=MaterialExpressionDivide_0}
    B {from=MaterialExpressionOneMinus_2}
   ]
  }
  MaterialExpressionMultiply_21 {
   class=MaterialExpressionMultiply
   inputs [
    A {from=MaterialExpressionMaterialFunctionCall_15, output=B}
    B {from=""}
   ]
   properties {constB=0.1}
  }
  MaterialExpressionMultiply_23 {
   class=MaterialExpressionMultiply
   inputs [
    A {from=MaterialExpressionTextureSample_5, output=G}
    B {from=""}
   ]
   properties {constB=10}
  }
  MaterialExpressionMultiply_24 {
   class=MaterialExpressionMultiply
   inputs [
    A {from=MaterialExpressionSine_1}
    B {from=""}
   ]
   properties {constB=0.5}
  }
  MaterialExpressionMultiply_25 {
   class=MaterialExpressionMultiply
   inputs [
    A {from=MaterialExpressionAdd_14}
    B {from=""}
   ]
   properties {constB=0.4}
  }
  MaterialExpressionMultiply_27 {
   class=MaterialExpressionMultiply
   inputs [
    A {from=MaterialExpressionMultiply_3}
    B {from=""}
   ]
   properties {constB=2}
  }
  MaterialExpressionMultiply_29 {
   class=MaterialExpressionMultiply
   inputs [
    A {from=MaterialExpressionFrac_2}
    B {from=""}
   ]
   properties {constB=4}
  }
  MaterialExpressionMultiply_3 {
   class=MaterialExpressionMultiply
   inputs [
    A {from=MaterialExpressionTime_1}
    B {from=MaterialExpressionDivide_6}
   ]
   properties {constB=2}
  }
  MaterialExpressionMultiply_30 {
   class=MaterialExpressionMultiply
   inputs [
    A {from=MaterialExpressionMultiply_4}
    B {from=""}
   ]
   properties {constB=0.5}
  }
  MaterialExpressionMultiply_31 {
   class=MaterialExpressionMultiply
   inputs [
    A {from=MaterialExpressionDesaturation_1}
    B {from=""}
   ]
   properties {constB=0.5}
  }
  MaterialExpressionMultiply_38 {
   class=MaterialExpressionMultiply
   inputs [
    A {from=MaterialExpressionMaterialFunctionCall_14, output=Amplitude}
    B {from=MaterialExpressionScalarParameter_26}
   ]
  }
  MaterialExpressionMultiply_4 {
   class=MaterialExpressionMultiply
   inputs [
    A {from=MaterialExpressionTime_0}
    B {from=MaterialExpressionDivide_8}
   ]
   properties {constB=2.1333}
  }
  MaterialExpressionMultiply_42 {
   class=MaterialExpressionMultiply
   inputs [
    A {from=MaterialExpressionMaterialFunctionCall_14, output=Spectrum}
    B {from=MaterialExpressionScalarParameter_26}
   ]
  }
  MaterialExpressionMultiply_5 {
   class=MaterialExpressionMultiply
   inputs [
    A {from=MaterialExpressionAdd_8}
    B {from=MaterialExpressionMax_0}
   ]
  }
  MaterialExpressionMultiply_54 {
   class=MaterialExpressionMultiply
   inputs [
    A {from=MaterialExpressionComponentMask_3}
    B {from=MaterialExpressionScalarParameter_24}
   ]
   properties {constB=6}
  }
  MaterialExpressionMultiply_55 {
   class=MaterialExpressionMultiply
   inputs [
    A {from=MaterialExpressionMultiply_57}
    B {from=MaterialExpressionScalarParameter_16}
   ]
  }
  MaterialExpressionMultiply_56 {
   class=MaterialExpressionMultiply
   inputs [
    A {from=MaterialExpressionComponentMask_4}
    B {from=""}
   ]
   properties {constB=0.1}
  }
  MaterialExpressionMultiply_57 {
   class=MaterialExpressionMultiply
   inputs [
    A {from=MaterialExpressionMaterialFunctionCall_21, output=Result}
    B {from=MaterialExpressionSaturate_6}
   ]
  }
  MaterialExpressionMultiply_58 {
   class=MaterialExpressionMultiply
   inputs [
    A {from=MaterialExpressionSaturate_6}
    B {from=MaterialExpressionSine_4}
   ]
  }
  MaterialExpressionMultiply_59 {
   class=MaterialExpressionMultiply
   inputs [
    A {from=MaterialExpressionMaterialFunctionCall_9, output=YX}
    B {from=""}
   ]
   properties {constB=2}
  }
  MaterialExpressionMultiply_6 {
   class=MaterialExpressionMultiply
   inputs [
    A {from=MaterialExpressionAbs_1}
    B {from=""}
   ]
   properties {constB=2}
  }
  MaterialExpressionMultiply_60 {
   class=MaterialExpressionMultiply
   inputs [
    A {from=MaterialExpressionMultiply_58}
    B {from=MaterialExpressionLinearInterpolate_1}
   ]
   properties {constB=0.2}
  }
  MaterialExpressionMultiply_61 {
   class=MaterialExpressionMultiply
   inputs [
    A {from=MaterialExpressionTime_5}
    B {from=MaterialExpressionScalarParameter_22}
   ]
   properties {constB=0.3111}
  }
  MaterialExpressionMultiply_65 {
   class=MaterialExpressionMultiply
   inputs [
    A {from=MaterialExpressionSphereMask_0}
    B {from=MaterialExpressionAdd_7}
   ]
  }
  MaterialExpressionMultiply_68 {
   class=MaterialExpressionMultiply
   inputs [
    A {from=MaterialExpressionNamedRerouteUsage_47}
    B {from=MaterialExpressionPower_6}
   ]
   properties {constA=1}
  }
  MaterialExpressionMultiply_7 {
   class=MaterialExpressionMultiply
   inputs [
    A {from=MaterialExpressionNamedRerouteUsage_4}
    B {from=MaterialExpressionMultiply_12}
   ]
   properties {constB=2}
  }
  MaterialExpressionMultiply_70 {
   class=MaterialExpressionMultiply
   inputs [
    A {from=MaterialExpressionOneMinus_19}
    B {from=""}
   ]
   properties {constB=50}
  }
  MaterialExpressionMultiply_71 {
   class=MaterialExpressionMultiply
   inputs [
    A {from=MaterialExpressionMultiply_56}
    B {from=MaterialExpressionMultiply_70}
   ]
  }
  MaterialExpressionMultiply_73 {
   class=MaterialExpressionMultiply
   inputs [
    A {from=MaterialExpressionFrac_3}
    B {from=""}
   ]
   properties {constB=2}
  }
  MaterialExpressionMultiply_74 {
   class=MaterialExpressionMultiply
   inputs [
    A {from=MaterialExpressionSine_6}
    B {from=""}
   ]
   properties {constB=0.5}
  }
  MaterialExpressionMultiply_75 {
   class=MaterialExpressionMultiply
   inputs [
    A {from=MaterialExpressionNamedRerouteUsage_34}
    B {from=""}
   ]
   properties {constB=0.75}
  }
  MaterialExpressionMultiply_77 {
   class=MaterialExpressionMultiply
   inputs [
    A {from=MaterialExpressionSaturate_7}
    B {from=MaterialExpressionAdd_8}
   ]
  }
  MaterialExpressionMultiply_8 {
   class=MaterialExpressionMultiply
   inputs [
    A {from=MaterialExpressionTextureSample_2, output=G}
    B {from=""}
   ]
   properties {constB=2}
  }
  MaterialExpressionMultiply_9 {
   class=MaterialExpressionMultiply
   inputs [
    A {from=MaterialExpressionAdd_3}
    B {from=MaterialExpressionLinearInterpolate_0}
   ]
   properties {constB=0.07}
  }
  MaterialExpressionNamedRerouteDeclaration_0 {
   class=MaterialExpressionNamedRerouteDeclaration
   inputs [None {from=MaterialExpressionMaterialFunctionCall_4, output=SquareMask}]
   properties {name=EqualizerMask, nodeColor=<rgba 0 0.6484 1 1>}
  }
  MaterialExpressionNamedRerouteDeclaration_1 {
   class=MaterialExpressionNamedRerouteDeclaration
   inputs [None {from=MaterialExpressionSaturate_3}]
   properties {name=PanelOutlinesMask, nodeColor=<rgba 1 0 0.5625 1>}
  }
  MaterialExpressionNamedRerouteDeclaration_10 {
   class=MaterialExpressionNamedRerouteDeclaration
   inputs [None {from=MaterialExpressionMultiply_2}]
   properties {name=BPM_HueShift, nodeColor=<rgba 0 1 0.8359 1>}
  }
  MaterialExpressionNamedRerouteDeclaration_11 {
   class=MaterialExpressionNamedRerouteDeclaration
   inputs [None {from=MaterialExpressionSaturate_2}]
   properties {name=Blink_Mask, nodeColor=<rgba 1 0.5156 0 1>}
  }
  MaterialExpressionNamedRerouteDeclaration_12 {
   class=MaterialExpressionNamedRerouteDeclaration
   inputs [None {from=MaterialExpressionGetMaterialAttributes_1, output=Opacity Mask}]
   properties {name=OriginalOpacity, nodeColor=<rgba 0.1953 0 1 1>}
  }
  MaterialExpressionNamedRerouteDeclaration_13 {
   class=MaterialExpressionNamedRerouteDeclaration
   inputs [None {from=MaterialExpressionMultiply_31}]
   properties {name=BaseColor_HueShifted, nodeColor=<rgba 0.125 1 0 1>}
  }
  MaterialExpressionNamedRerouteDeclaration_14 {
   class=MaterialExpressionNamedRerouteDeclaration
   inputs [None {from=MaterialExpressionGetMaterialAttributes_1, output=Emissive Color}]
   properties {name=OriginalEmissive, nodeColor=<rgba 1 0 0.4453 1>}
  }
  MaterialExpressionNamedRerouteDeclaration_15 {
   class=MaterialExpressionNamedRerouteDeclaration
   inputs [None {from=MaterialExpressionMultiply_77}]
   properties {name=OutputEmissive, nodeColor=<rgba 0 0.7656 1 1>}
  }
  MaterialExpressionNamedRerouteDeclaration_17 {
   class=MaterialExpressionNamedRerouteDeclaration
   inputs [None {from=MaterialExpressionAdd_26}]
   properties {name=BPM_Wave, nodeColor=<rgba 0.8516 1 0 1>}
  }
  MaterialExpressionNamedRerouteDeclaration_18 {
   class=MaterialExpressionNamedRerouteDeclaration
   inputs [None {from=MaterialExpressionMaterialFunctionCall_17, output=SquareMask}]
   properties {name=EqualizePanel_Mask, nodeColor=<rgba 0 0.6484 1 1>}
  }
  MaterialExpressionNamedRerouteDeclaration_2 {
   class=MaterialExpressionNamedRerouteDeclaration
   inputs [None {from=MaterialExpressionMaterialFunctionCall_0, output=Result}]
   properties {name=OriginalMA, nodeColor=<rgba 0 1 0.9297 1>}
  }
  MaterialExpressionNamedRerouteDeclaration_22 {
   class=MaterialExpressionNamedRerouteDeclaration
   inputs [None {from=MaterialExpressionMultiply_38}]
   properties {name=Amplitude, nodeColor=<rgba 0 1 0.1094 1>}
  }
  MaterialExpressionNamedRerouteDeclaration_23 {
   class=MaterialExpressionNamedRerouteDeclaration
   inputs [None {from=MaterialExpressionMultiply_42}]
   properties {name=Spectrum, nodeColor=<rgba 1 0 0.75 1>}
  }
  MaterialExpressionNamedRerouteDeclaration_24 {
   class=MaterialExpressionNamedRerouteDeclaration
   inputs [None {from=MaterialExpressionTextureSampleParameter2D_0, output=B}]
   properties {name=ArmEdgeSeamMask, nodeColor=<rgba 0.6641 0 1 1>}
  }
  MaterialExpressionNamedRerouteDeclaration_25 {
   class=MaterialExpressionNamedRerouteDeclaration
   inputs [None {from=MaterialExpressionPower_7}]
   properties {name=AmplitudeAdjusted, nodeColor=<rgba 0 1 0.9531 1>}
  }
  MaterialExpressionNamedRerouteDeclaration_26 {
   class=MaterialExpressionNamedRerouteDeclaration
   inputs [None {from=MaterialExpressionTextureSampleParameter2D_0, output=G}]
   properties {name=ArmpitsMask, nodeColor=<rgba 0 0.6484 1 1>}
  }
  MaterialExpressionNamedRerouteDeclaration_27 {
   class=MaterialExpressionNamedRerouteDeclaration
   inputs [None {from=MaterialExpressionTextureSampleParameter2D_0, output=R}]
   properties {name=ArmsMask, nodeColor=<rgba 1 0.5625 0 1>}
  }
  MaterialExpressionNamedRerouteDeclaration_28 {
   class=MaterialExpressionNamedRerouteDeclaration
   inputs [None {from=MaterialExpressionMultiply_14}]
   properties {name=NewEqualizer, nodeColor=<rgba 1 0.5391 0 1>}
  }
  MaterialExpressionNamedRerouteDeclaration_3 {
   class=MaterialExpressionNamedRerouteDeclaration
   inputs [None {from=MaterialExpressionTextureSample_0, output=R}]
   properties {name=EqualizerLine, nodeColor=<rgba 1 0.6094 0 1>}
  }
  MaterialExpressionNamedRerouteDeclaration_4 {
   class=MaterialExpressionNamedRerouteDeclaration
   inputs [None {from=MaterialExpressionPower_1}]
   properties {name=AudioAmplitude, nodeColor=<rgba 0.2891 0 1 1>}
  }
  MaterialExpressionNamedRerouteDeclaration_5 {
   class=MaterialExpressionNamedRerouteDeclaration
   inputs [None {from=MaterialExpressionMaterialFunctionCall_5, output=SquareMask}]
   properties {name=BottomDotsMask, nodeColor=<rgba 0.7109 0 1 1>}
  }
  MaterialExpressionNamedRerouteDeclaration_6 {
   class=MaterialExpressionNamedRerouteDeclaration
   inputs [None {from=MaterialExpressionMaterialFunctionCall_7, output=SquareMask}]
   properties {name=PanelsMask, nodeColor=<rgba 0.3828 0 1 1>}
  }
  MaterialExpressionNamedRerouteDeclaration_7 {
   class=MaterialExpressionNamedRerouteDeclaration
   inputs [None {from=MaterialExpressionMaterialFunctionCall_0, output=Tech Art Mask}]
   properties {name=TechArtMasks, nodeColor=<rgba 1 0.8203 0 1>}
  }
  MaterialExpressionNamedRerouteDeclaration_8 {
   class=MaterialExpressionNamedRerouteDeclaration
   inputs [None {from=MaterialExpressionSetMaterialAttributes_1}]
   properties {name=OutputMA, nodeColor=<rgba 0.5234 1 0 1>}
  }
  MaterialExpressionNamedRerouteDeclaration_9 {
   class=MaterialExpressionNamedRerouteDeclaration
   inputs [None {from=MaterialExpressionMultiply_29}]
   properties {name=BPM_Blink, nodeColor=<rgba 1 0 0.8437 1>}
  }
  MaterialExpressionNamedRerouteUsage_10 {class=MaterialExpressionNamedRerouteUsage}
  MaterialExpressionNamedRerouteUsage_11 {class=MaterialExpressionNamedRerouteUsage}
  MaterialExpressionNamedRerouteUsage_12 {class=MaterialExpressionNamedRerouteUsage}
  MaterialExpressionNamedRerouteUsage_13 {class=MaterialExpressionNamedRerouteUsage}
  MaterialExpressionNamedRerouteUsage_14 {class=MaterialExpressionNamedRerouteUsage}
  MaterialExpressionNamedRerouteUsage_15 {class=MaterialExpressionNamedRerouteUsage}
  MaterialExpressionNamedRerouteUsage_16 {class=MaterialExpressionNamedRerouteUsage}
  MaterialExpressionNamedRerouteUsage_17 {class=MaterialExpressionNamedRerouteUsage}
  MaterialExpressionNamedRerouteUsage_18 {class=MaterialExpressionNamedRerouteUsage}
  MaterialExpressionNamedRerouteUsage_19 {class=MaterialExpressionNamedRerouteUsage}
  MaterialExpressionNamedRerouteUsage_2 {class=MaterialExpressionNamedRerouteUsage}
  MaterialExpressionNamedRerouteUsage_20 {class=MaterialExpressionNamedRerouteUsage}
  MaterialExpressionNamedRerouteUsage_24 {class=MaterialExpressionNamedRerouteUsage}
  MaterialExpressionNamedRerouteUsage_25 {class=MaterialExpressionNamedRerouteUsage}
  MaterialExpressionNamedRerouteUsage_26 {class=MaterialExpressionNamedRerouteUsage}
  MaterialExpressionNamedRerouteUsage_27 {class=MaterialExpressionNamedRerouteUsage}
  MaterialExpressionNamedRerouteUsage_3 {class=MaterialExpressionNamedRerouteUsage}
  MaterialExpressionNamedRerouteUsage_34 {class=MaterialExpressionNamedRerouteUsage}
  MaterialExpressionNamedRerouteUsage_35 {class=MaterialExpressionNamedRerouteUsage}
  MaterialExpressionNamedRerouteUsage_38 {class=MaterialExpressionNamedRerouteUsage}
  MaterialExpressionNamedRerouteUsage_4 {class=MaterialExpressionNamedRerouteUsage}
  MaterialExpressionNamedRerouteUsage_46 {class=MaterialExpressionNamedRerouteUsage}
  MaterialExpressionNamedRerouteUsage_47 {class=MaterialExpressionNamedRerouteUsage}
  MaterialExpressionNamedRerouteUsage_48 {class=MaterialExpressionNamedRerouteUsage}
  MaterialExpressionNamedRerouteUsage_5 {class=MaterialExpressionNamedRerouteUsage}
  MaterialExpressionNamedRerouteUsage_7 {class=MaterialExpressionNamedRerouteUsage}
  MaterialExpressionNamedRerouteUsage_8 {class=MaterialExpressionNamedRerouteUsage}
  MaterialExpressionNamedRerouteUsage_9 {class=MaterialExpressionNamedRerouteUsage}
  MaterialExpressionOneMinus_0 {class=MaterialExpressionOneMinus, inputs [None {from=MaterialExpressionMultiply_56}]}
  MaterialExpressionOneMinus_1 {class=MaterialExpressionOneMinus, inputs [None {from=MaterialExpressionComponentMask_1}]}
  MaterialExpressionOneMinus_14 {class=MaterialExpressionOneMinus, inputs [None {from=MaterialExpressionMultiply_65}]}
  MaterialExpressionOneMinus_15 {class=MaterialExpressionOneMinus, inputs [None {from=MaterialExpressionLinearInterpolate_7}]}
  MaterialExpressionOneMinus_16 {class=MaterialExpressionOneMinus, inputs [None {from=MaterialExpressionDivide_2}]}
  MaterialExpressionOneMinus_19 {class=MaterialExpressionOneMinus, inputs [None {from=MaterialExpressionAdd_24}]}
  MaterialExpressionOneMinus_2 {class=MaterialExpressionOneMinus, inputs [None {from=MaterialExpressionNamedRerouteUsage_26}]}
  MaterialExpressionOneMinus_20 {class=MaterialExpressionOneMinus, inputs [None {from=MaterialExpressionNamedRerouteUsage_27}]}
  MaterialExpressionOneMinus_3 {class=MaterialExpressionOneMinus, inputs [None {from=MaterialExpressionTextureSample_5, output=R}]}
  MaterialExpressionOneMinus_4 {class=MaterialExpressionOneMinus, inputs [None {from=MaterialExpressionTextureSample_5, output=B}]}
  MaterialExpressionOneMinus_5 {class=MaterialExpressionOneMinus, inputs [None {from=MaterialExpressionNamedRerouteUsage_5}]}
  MaterialExpressionPanner_0 {
   class=MaterialExpressionPanner
   inputs [
    Coordinate {from=MaterialExpressionMaterialFunctionCall_8, output=UVs}
    Time {from=MaterialExpressionSine_0}
    Speed {from=""}
   ]
   properties {speedY=0.05}
  }
  MaterialExpressionPower_0 {
   class=MaterialExpressionPower
   inputs [
    Base {from=MaterialExpressionNamedRerouteUsage_3}
    Exp {from=""}
   ]
  }
  MaterialExpressionPower_1 {
   class=MaterialExpressionPower
   inputs [
    Base {from=MaterialExpressionMaterialFunctionCall_3, output=Amplitude}
    Exp {from=""}
   ]
  }
  MaterialExpressionPower_2 {
   class=MaterialExpressionPower
   inputs [
    Base {from=MaterialExpressionNamedRerouteUsage_35}
    Exp {from=MaterialExpressionScalarParameter_2}
   ]
   properties {constExponent=4}
  }
  MaterialExpressionPower_3 {
   class=MaterialExpressionPower
   inputs [
    Base {from=MaterialExpressionNamedRerouteUsage_8}
    Exp {from=""}
   ]
   properties {constExponent=12}
  }
  MaterialExpressionPower_6 {
   class=MaterialExpressionPower
   inputs [
    Base {from=MaterialExpressionOneMinus_16}
    Exp {from=MaterialExpressionScalarParameter_5}
   ]
   properties {constExponent=10}
  }
  MaterialExpressionPower_7 {
   class=MaterialExpressionPower
   inputs [
    Base {from=MaterialExpressionNamedRerouteUsage_38}
    Exp {from=MaterialExpressionScalarParameter_14}
   ]
   properties {constExponent=8}
  }
  MaterialExpressionReroute_0 {class=MaterialExpressionReroute, inputs [None {from=MaterialExpressionGetMaterialAttributes_1, output=MaterialAttributes}]}
  MaterialExpressionReroute_1 {class=MaterialExpressionReroute, inputs [None {from=MaterialExpressionReroute_0}]}
  MaterialExpressionReroute_2 {class=MaterialExpressionReroute, inputs [None {from=MaterialExpressionMaterialFunctionCall_21, output=Result}]}
  MaterialExpressionSaturate_0 {class=MaterialExpressionSaturate, inputs [None {from=MaterialExpressionAdd_5}]}
  MaterialExpressionSaturate_2 {class=MaterialExpressionSaturate, inputs [None {from=MaterialExpressionAdd_9}]}
  MaterialExpressionSaturate_3 {class=MaterialExpressionSaturate, inputs [None {from=MaterialExpressionMultiply_23}]}
  MaterialExpressionSaturate_4 {class=MaterialExpressionSaturate, inputs [None {from=MaterialExpressionSubtract_1}]}
  MaterialExpressionSaturate_5 {class=MaterialExpressionSaturate, inputs [None {from=MaterialExpressionAdd_10}]}
  MaterialExpressionSaturate_6 {class=MaterialExpressionSaturate, inputs [None {from=MaterialExpressionOneMinus_0}]}
  MaterialExpressionSaturate_7 {class=MaterialExpressionSaturate, inputs [None {from=MaterialExpressionAdd_27}]}
  MaterialExpressionScalarParameter_0 {class=MaterialExpressionScalarParameter, properties {defaultValue=0.25, group=FareSpore, parameterName=SliderValue_01, sliderMax=1}}
  MaterialExpressionScalarParameter_1 {class=MaterialExpressionScalarParameter, properties {defaultValue=0.75, group=FareSpore, parameterName=SliderValue_02, sliderMax=1}}
  MaterialExpressionScalarParameter_13 {
   class=MaterialExpressionScalarParameter
   properties {defaultValue=2.1333, group=TechArt, parameterName=StringFreqCurveFactor, sortPriority=40}
  }
  MaterialExpressionScalarParameter_14 {class=MaterialExpressionScalarParameter, properties {defaultValue=8, group=TechArt, parameterName=AmplitudeEmissiveExp}}
  MaterialExpressionScalarParameter_16 {class=MaterialExpressionScalarParameter, properties {defaultValue=1.8, group=TechArt, parameterName=StringExciteAmt}}
  MaterialExpressionScalarParameter_2 {class=MaterialExpressionScalarParameter, properties {defaultValue=2, parameterName=SpectrumExp}}
  MaterialExpressionScalarParameter_22 {
   class=MaterialExpressionScalarParameter
   properties {defaultValue=0.3, group=TechArt, parameterName=StringBaseSineFrequency, sortPriority=40}
  }
  MaterialExpressionScalarParameter_24 {class=MaterialExpressionScalarParameter, properties {defaultValue=10, group=TechArt, parameterName=StringGradientLength}}
  MaterialExpressionScalarParameter_26 {class=MaterialExpressionScalarParameter, properties {defaultValue=1, group=TechArt, parameterName=EnableAudioReactivity, sliderMax=1}}
  MaterialExpressionScalarParameter_27 {class=MaterialExpressionScalarParameter, properties {group=TechArt, parameterName=StringBaseSineAmplitude, sortPriority=40}}
  MaterialExpressionScalarParameter_29 {class=MaterialExpressionScalarParameter, properties {defaultValue=0.32, group=TechArt, parameterName=StringAmpBase, sortPriority=40}}
  MaterialExpressionScalarParameter_3 {class=MaterialExpressionScalarParameter, properties {defaultValue=0.3, group=TechArt, parameterName=StringThickness, sortPriority=38}}
  MaterialExpressionScalarParameter_30 {class=MaterialExpressionScalarParameter, properties {defaultValue=1, group=TechArt, parameterName=StringAmpCurveFactor, sortPriority=40}}
  MaterialExpressionScalarParameter_31 {class=MaterialExpressionScalarParameter, properties {defaultValue=3, group=TechArt, parameterName=StringScrollBase, sortPriority=40}}
  MaterialExpressionScalarParameter_32 {
   class=MaterialExpressionScalarParameter
   properties {defaultValue=0.625, group=TechArt, parameterName=StringScrollCurveFactor, sortPriority=40}
  }
  MaterialExpressionScalarParameter_35 {class=MaterialExpressionScalarParameter, properties {defaultValue=0.4, group=TechArt, parameterName=StringOffset, sortPriority=38}}
  MaterialExpressionScalarParameter_36 {class=MaterialExpressionScalarParameter, properties {defaultValue=0.5, group=TechArt, parameterName=StringFreqBase, sortPriority=40}}
  MaterialExpressionScalarParameter_4 {
   class=MaterialExpressionScalarParameter
   properties {defaultValue=0.04, group=TechArt, parameterName=StringSeparation, sliderMax=0.5, sortPriority=38}
  }
  MaterialExpressionScalarParameter_5 {class=MaterialExpressionScalarParameter, properties {defaultValue=60, group=TechArt, parameterName=StringThicknessExp, sortPriority=38}}
  MaterialExpressionSetMaterialAttributes_1 {
   class=MaterialExpressionSetMaterialAttributes
   inputs [
    MaterialAttributes {from=MaterialExpressionReroute_1, output=""}
    Emissive Color {from=MaterialExpressionNamedRerouteUsage_20, output=""}
    Base Color {from=MaterialExpressionNamedRerouteUsage_18, output=""}
    Opacity Mask {from=MaterialExpressionLinearInterpolate_2, output=""}
   ]
   properties {attributeSetTypes [#1, #2, #3]}
  }
  MaterialExpressionSine_0 {class=MaterialExpressionSine, inputs [None {from=MaterialExpressionMultiply_3}]}
  MaterialExpressionSine_1 {class=MaterialExpressionSine, inputs [None {from=MaterialExpressionMultiply_27}]}
  MaterialExpressionSine_4 {class=MaterialExpressionSine, inputs [None {from=MaterialExpressionAdd_17}]}
  MaterialExpressionSine_6 {class=MaterialExpressionSine, inputs [None {from=MaterialExpressionMultiply_4}]}
  MaterialExpressionSphereMask_0 {
   class=MaterialExpressionSphereMask
   inputs [
    A {from=MaterialExpressionAdd_2, output=""}
    B {from=MaterialExpressionScalarParameter_35, output=""}
    Radius {from=MaterialExpressionScalarParameter_3, output=""}
    Hardness {from="", output=""}
   ]
   properties {attenuationRadius=1, hardnessPercent=0}
  }
  MaterialExpressionStaticBool_1 {class=MaterialExpressionStaticBool, properties {value=true}}
  MaterialExpressionSubtract_0 {
   class=MaterialExpressionSubtract
   inputs [
    A {from=MaterialExpressionMaterialFunctionCall_8, output=UVs}
    B {from=MaterialExpressionMultiply_9}
   ]
  }
  MaterialExpressionSubtract_1 {
   class=MaterialExpressionSubtract
   inputs [
    A {from=MaterialExpressionMaterialFunctionCall_3, output=Amplitude}
    B {from=MaterialExpressionMultiply_25}
   ]
  }
  MaterialExpressionSubtract_3 {
   class=MaterialExpressionSubtract
   inputs [
    A {from=MaterialExpressionAbs_4}
    B {from=MaterialExpressionLinearInterpolate_7}
   ]
   properties {constB=0.05}
  }
  MaterialExpressionSubtract_8 {
   class=MaterialExpressionSubtract
   inputs [
    A {from=MaterialExpressionMultiply_59}
    B {from=""}
   ]
   properties {constB=0.98}
  }
  MaterialExpressionSwitch_0 {
   class=MaterialExpressionSwitch
   inputs [
    SwitchValue {from=MaterialExpressionNamedRerouteUsage_15, output=""}
    Default {from=MaterialExpressionTextureSample_5, output=R}
    None {from=MaterialExpressionTextureSample_5, output=R}
    None {from=MaterialExpressionTextureSample_5, output=R}
    None {from=MaterialExpressionOneMinus_3, output=""}
    None {from=MaterialExpressionOneMinus_4, output=""}
   ]
   properties {inputs [
     {
      inputName=None
      input {
       expression=&self:MaterialExpressionTextureSample_5
       inputName=None
       mask=1
       maskA=0
       maskB=0
       maskG=0
       maskR=1
       outputIndex=1
      }
     }
     {
      inputName=None
      input {
       expression=&self:MaterialExpressionTextureSample_5
       inputName=None
       mask=1
       maskA=0
       maskB=1
       maskG=0
       maskR=0
       outputIndex=3
      }
     }
     {
      inputName=None
      input {
       expression=&self:MaterialExpressionOneMinus_3
       inputName=None
       mask=0
       maskA=0
       maskB=0
       maskG=0
       maskR=0
       outputIndex=0
      }
     }
     {
      inputName=None
      input {
       expression=&self:MaterialExpressionOneMinus_4
       inputName=None
       mask=0
       maskA=0
       maskB=0
       maskG=0
       maskR=0
       outputIndex=0
      }
     }
    ]}
  }
  MaterialExpressionTextureCoordinate_0 {class=MaterialExpressionTextureCoordinate}
  MaterialExpressionTextureCoordinate_1 {class=MaterialExpressionTextureCoordinate}
  MaterialExpressionTextureCoordinate_10 {class=MaterialExpressionTextureCoordinate}
  MaterialExpressionTextureCoordinate_11 {class=MaterialExpressionTextureCoordinate}
  MaterialExpressionTextureCoordinate_12 {class=MaterialExpressionTextureCoordinate}
  MaterialExpressionTextureCoordinate_2 {class=MaterialExpressionTextureCoordinate}
  MaterialExpressionTextureCoordinate_3 {class=MaterialExpressionTextureCoordinate}
  MaterialExpressionTextureCoordinate_6 {class=MaterialExpressionTextureCoordinate}
  MaterialExpressionTextureSampleParameter2D_0 {
   class=MaterialExpressionTextureSampleParameter2D
   inputs [
    UVs {from="", output=""}
    Apply View MipBias {from="", output=""}
   ]
   properties {
    parameterName=TechArtFX_Mask
    samplerSource=SSM_Wrap_WorldGroupSettings
    samplerType=SAMPLERTYPE_LinearColor
    texture=/BRCosmetics/Characters/Player/Female/Medium/Bodies/F_MED_Tactical_Red_Punk/Skins/Disco/Textures/T_TacticalRed_Disco_body_TechArtFX_Mask
   }
  }
  MaterialExpressionTextureSample_0 {
   class=MaterialExpressionTextureSample
   inputs [
    UVs {from=MaterialExpressionSubtract_0}
    Tex {from=""}
    Apply View MipBias {from=""}
   ]
   properties {
    samplerSource=SSM_Clamp_WorldGroupSettings
    texture=/BRCosmetics/Animation/Props/Gadgets_FareSporeMilk/Textures/T_FareSpore_Milk_AdditionalFX
   }
  }
  MaterialExpressionTextureSample_2 {
   class=MaterialExpressionTextureSample
   inputs [
    UVs {from=MaterialExpressionPanner_0}
    Tex {from=""}
    Apply View MipBias {from=""}
   ]
   properties {samplerSource=SSM_Wrap_WorldGroupSettings, texture=&T_FareSpore_Milk_AdditionalFX}
  }
  MaterialExpressionTextureSample_5 {
   class=MaterialExpressionTextureSample
   inputs [
    UVs {from=""}
    Tex {from=""}
    Apply View MipBias {from=""}
   ]
   properties {samplerSource=SSM_Wrap_WorldGroupSettings, texture=/BRCosmetics/Animation/Props/Gadgets_FareSporeMilk/Textures/T_FareSpore_Milk_PannelFX}
  }
  MaterialExpressionTime_0 {class=MaterialExpressionTime}
  MaterialExpressionTime_1 {class=MaterialExpressionTime}
  MaterialExpressionTime_5 {class=MaterialExpressionTime}
 }
 properties {
  allowTranslucentCustomDepthWrites=true
  bUseMaterialAttributes=true
  bUsedWithMeshDeformer=true
  bUsedWithMorphTargets=true
  bUsedWithNiagaraMeshParticles=true
  bUsedWithSkeletalMesh=true
  blendMode=BLEND_Masked
  pixelDepthOffsetMode=PDOM_Legacy
  shadingModel=MSM_FromMaterialExpression
  usedShadingModels=Default Lit | Subsurface | Clear Coat | Subsurface Profile | Cloth
 }
}
